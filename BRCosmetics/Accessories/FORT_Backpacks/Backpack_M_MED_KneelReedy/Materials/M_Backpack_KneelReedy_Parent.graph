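# Parent material for the KneelReedy backpack cosmetic, layering an animated portal with ghosts over the shared 2023 anime character shader.
# Radial swirl UVs, panning distortion and vignette shape the tileable portal blob, composited with three scaled, distorted ghost textures driven by a Ghost Phase parameter.
# Sine-pulsed inverted Fresnel adds reactive emissive light, while normal-based vertex offset is masked.
# A scalar toggle picks the mask that blends these effects over the base attributes.
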Material /BRCosmetics/Accessories/FORT_Backpacks/Backpack_M_MED_KneelReedy/Materials/M_Backpack_KneelReedy_Parent
graph {
 expressions {
  MaterialExpressionAdd_15 {
   class=MaterialExpressionAdd
   inputs [
    A {from=MaterialExpressionTextureCoordinate_7}
    B {from=MaterialExpressionComponentMask_33}
   ]
  }
  MaterialExpressionAdd_16 {
   class=MaterialExpressionAdd
   inputs [
    A {from=MaterialExpressionNamedRerouteUsage_15}
    B {from=MaterialExpressionMultiply_36}
   ]
  }
  MaterialExpressionAdd_17 {
   class=MaterialExpressionAdd
   inputs [
    A {from=MaterialExpressionMultiply_39}
    B {from=MaterialExpressionMultiply_40}
   ]
  }
  MaterialExpressionAdd_18 {
   class=MaterialExpressionAdd
   inputs [
    A {from=MaterialExpressionMultiply_45}
    B {from=MaterialExpressionMultiply_47}
   ]
  }
  MaterialExpressionAdd_19 {
   class=MaterialExpressionAdd
   inputs [
    A {from=MaterialExpressionNamedRerouteDeclaration_31}
    B {from=MaterialExpressionComponentMask_39}
   ]
  }
  MaterialExpressionAdd_21 {
   class=MaterialExpressionAdd
   inputs [
    A {from=MaterialExpressionScalarParameter_44}
    B {from=MaterialExpressionScalarParameter_45}
   ]
  }
  MaterialExpressionAdd_23 {
   class=MaterialExpressionAdd
   inputs [
    A {from=MaterialExpressionSubtract_13}
    B {from=MaterialExpressionComponentMask_48}
   ]
  }
  MaterialExpressionAdd_24 {
   class=MaterialExpressionAdd
   inputs [
    A {from=MaterialExpressionSubtract_14}
    B {from=MaterialExpressionComponentMask_50}
   ]
  }
  MaterialExpressionAdd_25 {
   class=MaterialExpressionAdd
   inputs [
    A {from=MaterialExpressionSubtract_11}
    B {from=MaterialExpressionMultiply_67}
   ]
  }
  MaterialExpressionAdd_26 {
   class=MaterialExpressionAdd
   inputs [
    A {from=MaterialExpressionSubtract_12}
    B {from=MaterialExpressionComponentMask_56}
   ]
  }
  MaterialExpressionAppendVector_1 {
   class=MaterialExpressionAppendVector
   inputs [
    A {from=MaterialExpressionVertexColor_1, output=""}
    B {from=MaterialExpressionVertexColor_1, output=""}
   ]
  }
  MaterialExpressionAppendVector_2 {
   class=MaterialExpressionAppendVector
   inputs [
    A {from=MaterialExpressionComponentMask_54, output=""}
    B {from=MaterialExpressionMultiply_73, output=""}
   ]
  }
  MaterialExpressionBlendMaterialAttributes_1 {
   class=MaterialExpressionBlendMaterialAttributes
   inputs [
    A {from=MaterialExpressionNamedRerouteUsage_39, output=""}
    B {from=MaterialExpressionSetMaterialAttributes_2, output=""}
    Alpha {from=MaterialExpressionLinearInterpolate_16, output=""}
   ]
   properties {}
  }
  MaterialExpressionCeil_3 {class=MaterialExpressionCeil, inputs [None {from=MaterialExpressionSubtract_4, output=""}], properties {}}
  MaterialExpressionCeil_4 {class=MaterialExpressionCeil, inputs [None {from=MaterialExpressionSubtract_5, output=""}], properties {}}
  MaterialExpressionChannelMaskParameter_1 {
   class=MaterialExpressionChannelMaskParameter
   inputs [None {from=MaterialExpressionStaticSwitchParameter_5, output=""}]
   properties {group=01_EyeAnimation_Parameters, parameterName=EyeMaskChannel, sortPriority=-32}
  }
  MaterialExpressionComponentMask_33 {class=MaterialExpressionComponentMask, inputs [None {from=MaterialExpressionVectorParameter_22, output=RGB}], properties {g=true, r=true}}
  MaterialExpressionComponentMask_34 {class=MaterialExpressionComponentMask, inputs [None {from=MaterialExpressionVectorParameter_17, output=RGB}], properties {g=true, r=true}}
  MaterialExpressionComponentMask_35 {class=MaterialExpressionComponentMask, inputs [None {from=MaterialExpressionNamedRerouteUsage_22}], properties {b=true}}
  MaterialExpressionComponentMask_36 {class=MaterialExpressionComponentMask, inputs [None {from=MaterialExpressionScalarParameter_42}], properties {g=true, r=true}}
  MaterialExpressionComponentMask_37 {class=MaterialExpressionComponentMask, inputs [None {from=MaterialExpressionNamedRerouteUsage_24}], properties {g=true, r=true}}
  MaterialExpressionComponentMask_38 {class=MaterialExpressionComponentMask, inputs [None {from=MaterialExpressionVectorParameter_18, output=RGB}], properties {g=true, r=true}}
  MaterialExpressionComponentMask_39 {class=MaterialExpressionComponentMask, inputs [None {from=MaterialExpressionVectorParameter_21, output=RGB}], properties {g=true, r=true}}
  MaterialExpressionComponentMask_4 {class=MaterialExpressionComponentMask, inputs [None {from=MaterialExpressionTextureCoordinate_5}], properties {r=true}}
  MaterialExpressionComponentMask_40 {class=MaterialExpressionComponentMask, inputs [None {from=MaterialExpressionNamedRerouteUsage_50}], properties {b=true, g=true, r=true}}
  MaterialExpressionComponentMask_41 {class=MaterialExpressionComponentMask, inputs [None {from=MaterialExpressionNamedRerouteUsage_50}], properties {a=true}}
  MaterialExpressionComponentMask_42 {class=MaterialExpressionComponentMask, inputs [None {from=MaterialExpressionNamedRerouteUsage_27}], properties {r=true}}
  MaterialExpressionComponentMask_46 {class=MaterialExpressionComponentMask, inputs [None {from=MaterialExpressionVectorParameter_25, output=RGB}], properties {g=true, r=true}}
  MaterialExpressionComponentMask_47 {class=MaterialExpressionComponentMask, inputs [None {from=MaterialExpressionVectorParameter_28, output=RGB}], properties {g=true, r=true}}
  MaterialExpressionComponentMask_48 {class=MaterialExpressionComponentMask, inputs [None {from=MaterialExpressionVectorParameter_29, output=RGB}], properties {g=true, r=true}}
  MaterialExpressionComponentMask_49 {class=MaterialExpressionComponentMask, inputs [None {from=MaterialExpressionNamedRerouteUsage_48}], properties {a=true}}
  MaterialExpressionComponentMask_50 {class=MaterialExpressionComponentMask, inputs [None {from=MaterialExpressionVectorParameter_31, output=RGB}], properties {g=true, r=true}}
  MaterialExpressionComponentMask_51 {class=MaterialExpressionComponentMask, inputs [None {from=MaterialExpressionNamedRerouteUsage_49}], properties {a=true}}
  MaterialExpressionComponentMask_52 {class=MaterialExpressionComponentMask, inputs [None {from=MaterialExpressionTextureSample_8, output=RGB}], properties {g=true, r=true}}
  MaterialExpressionComponentMask_53 {class=MaterialExpressionComponentMask, inputs [None {from=MaterialExpressionVectorParameter_35, output=RGB}], properties {g=true, r=true}}
  MaterialExpressionComponentMask_54 {class=MaterialExpressionComponentMask, inputs [None {from=MaterialExpressionConvert_2}], properties {r=true}}
  MaterialExpressionComponentMask_55 {class=MaterialExpressionComponentMask, inputs [None {from=MaterialExpressionConvert_2}], properties {g=true}}
  MaterialExpressionComponentMask_56 {class=MaterialExpressionComponentMask, inputs [None {from=MaterialExpressionVectorParameter_36, output=RGB}], properties {g=true, r=true}}
  MaterialExpressionConstant_1 {class=MaterialExpressionConstant, properties {r=0.99}}
  MaterialExpressionConstant_10 {class=MaterialExpressionConstant, properties {r=0.1}}
  MaterialExpressionConstant_6 {class=MaterialExpressionConstant, properties {r=1}}
  MaterialExpressionConstant_7 {class=MaterialExpressionConstant}
  MaterialExpressionConstant_8 {class=MaterialExpressionConstant, properties {r=0.5}}
  MaterialExpressionConstant_9 {class=MaterialExpressionConstant}
  MaterialExpressionConvert_2 {
   class=MaterialExpressionConvert
   inputs [
    None {from=MaterialExpressionConvert_3}
    None {from=MaterialExpressionSubtract_8}
   ]
   properties {
    nodeName=Make Vector2
    convertInputs [
     {
      type=Scalar
      expressionInput {
       expression=&self:MaterialExpressionConvert_3
       inputName=None
       mask=0
       maskA=0
       maskB=0
       maskG=0
       maskR=0
       outputIndex=0
      }
     }
     {
      type=Scalar
      expressionInput {
       expression=&self:MaterialExpressionSubtract_8
       inputName=None
       mask=0
       maskA=0
       maskB=0
       maskG=0
       maskR=0
       outputIndex=0
      }
     }
    ]
    convertOutputs [{type=Vector2}]
   }
  }
  MaterialExpressionConvert_3 {
   class=MaterialExpressionConvert
   inputs [None {from=MaterialExpressionPower_8}]
   properties {
    nodeName=Break Float2
    convertInputs [{
      type=Vector2
      expressionInput {
       expression=&self:MaterialExpressionPower_8
       inputName=None
       mask=0
       maskA=0
       maskB=0
       maskG=0
       maskR=0
       outputIndex=0
      }
     }]
    convertOutputs [{type=Scalar}, {type=Scalar}]
   }
  }
  MaterialExpressionDistance_0 {
   class=MaterialExpressionDistance
   inputs [
    A {from=MaterialExpressionSubtract_10, output=""}
    B {from=MaterialExpressionConstant_9, output=""}
   ]
   properties {}
  }
  MaterialExpressionDivide_1 {
   class=MaterialExpressionDivide
   inputs [
    A {from=MaterialExpressionConvert_3}
    B {from=MaterialExpressionScalarParameter_57}
   ]
  }
  MaterialExpressionFrac_0 {class=MaterialExpressionFrac, inputs [None {from=MaterialExpressionAdd_26, output=""}], properties {}}
  MaterialExpressionFrac_1 {class=MaterialExpressionFrac, inputs [None {from=MaterialExpressionAdd_23, output=""}], properties {}}
  MaterialExpressionFrac_16 {class=MaterialExpressionFrac, inputs [None {from=MaterialExpressionAdd_24, output=""}], properties {}}
  MaterialExpressionFresnel_1 {
   class=MaterialExpressionFresnel
   inputs [
    ExponentIn {from=MaterialExpressionScalarParameter_35, output=""}
    BaseReflectFractionIn {from="", output=""}
    Normal {from="", output=""}
   ]
   properties {}
  }
  MaterialExpressionGetMaterialAttributes_0 {
   class=MaterialExpressionGetMaterialAttributes
   inputs [None {from=MaterialExpressionMaterialFunctionCall_2, output=Result}]
   properties {attributeGetTypes [#1, #2]}
  }
  MaterialExpressionLinearInterpolate_0 {
   class=MaterialExpressionLinearInterpolate
   inputs [
    A {from=""}
    B {from=MaterialExpressionTextureSample_2, output=RGB}
    Alpha {from=MaterialExpressionStep_1}
   ]
  }
  MaterialExpressionLinearInterpolate_10 {
   class=MaterialExpressionLinearInterpolate
   inputs [
    A {from=""}
    B {from=MaterialExpressionMultiply_60}
    Alpha {from=MaterialExpressionNamedRerouteUsage_38}
   ]
   properties {constB=0}
  }
  MaterialExpressionLinearInterpolate_12 {
   class=MaterialExpressionLinearInterpolate
   inputs [
    A {from=MaterialExpressionNamedRerouteUsage_15}
    B {from=MaterialExpressionAdd_16}
    Alpha {from=MaterialExpressionNamedRerouteUsage_37}
   ]
  }
  MaterialExpressionLinearInterpolate_13 {
   class=MaterialExpressionLinearInterpolate
   inputs [
    A {from=""}
    B {from=MaterialExpressionMaterialFunctionCall_18, output=Remapped Sine}
    Alpha {from=MaterialExpressionLinearInterpolate_14}
   ]
  }
  MaterialExpressionLinearInterpolate_14 {
   class=MaterialExpressionLinearInterpolate
   inputs [
    A {from=""}
    B {from=MaterialExpressionScalarParameter_40}
    Alpha {from=MaterialExpressionScalarParameter_41}
   ]
   properties {constB=0}
  }
  MaterialExpressionLinearInterpolate_15 {
   class=MaterialExpressionLinearInterpolate
   inputs [
    A {from=MaterialExpressionMultiply_49}
    B {from=MaterialExpressionComponentMask_40}
    Alpha {from=MaterialExpressionSaturate_13}
   ]
  }
  MaterialExpressionLinearInterpolate_16 {
   class=MaterialExpressionLinearInterpolate
   inputs [
    A {from=MaterialExpressionNamedRerouteUsage_44}
    B {from=MaterialExpressionNamedRerouteUsage_45}
    Alpha {from=MaterialExpressionScalarParameter_49}
   ]
  }
  MaterialExpressionLinearInterpolate_17 {
   class=MaterialExpressionLinearInterpolate
   inputs [
    A {from=MaterialExpressionMultiply_58}
    B {from=MaterialExpressionMultiply_57}
    Alpha {from=MaterialExpressionNamedRerouteUsage_47}
   ]
  }
  MaterialExpressionLinearInterpolate_18 {
   class=MaterialExpressionLinearInterpolate
   inputs [
    A {from=MaterialExpressionNamedRerouteDeclaration_36}
    B {from=MaterialExpressionNamedRerouteUsage_48}
    Alpha {from=MaterialExpressionComponentMask_49}
   ]
  }
  MaterialExpressionLinearInterpolate_19 {
   class=MaterialExpressionLinearInterpolate
   inputs [
    A {from=MaterialExpressionLinearInterpolate_18}
    B {from=MaterialExpressionNamedRerouteUsage_49}
    Alpha {from=MaterialExpressionComponentMask_51}
   ]
  }
  MaterialExpressionLinearInterpolate_20 {
   class=MaterialExpressionLinearInterpolate
   inputs [
    A {from=MaterialExpressionScalarParameter_63}
    B {from=MaterialExpressionScalarParameter_64}
    Alpha {from=MaterialExpressionScalarParameter_61}
   ]
  }
  MaterialExpressionLinearInterpolate_21 {
   class=MaterialExpressionLinearInterpolate
   inputs [
    A {from=MaterialExpressionConstant_10}
    B {from=""}
    Alpha {from=MaterialExpressionScalarParameter_74}
   ]
  }
  MaterialExpressionMaterialFunctionCall_10 {
   class=MaterialExpressionMaterialFunctionCall
   inputs [CustomUV {from=""}]
   properties {materialFunction=/Game/Global/Materials/Character_Parent_2023/Anime_Parent_2023/MaterialFunctions/MFC_AnimeShader_2023_Pre}
  }
  MaterialExpressionMaterialFunctionCall_11 {
   class=MaterialExpressionMaterialFunctionCall
   inputs [
    Color Lit {from=MaterialExpressionMaterialFunctionCall_10, output=Color_Lit}
    Color Shaded {from=MaterialExpressionMaterialFunctionCall_10, output=Color_Lit}
    Use UV2 {from=MaterialExpressionStaticBoolParameter_1}
    Mask {from=MaterialExpressionStaticSwitchParameter_3}
   ]
   properties {materialFunction=/Game/Global/Materials/Character_Parent_2023/Anime_Parent_2023/MaterialFunctions/MF_AnimeEyeAnimation_v2}
  }
  MaterialExpressionMaterialFunctionCall_17 {
   class=MaterialExpressionMaterialFunctionCall
   inputs [
    UVs {from=MaterialExpressionAdd_15}
    Texture Scale {from=MaterialExpressionScalarParameter_34}
   ]
   properties {materialFunction=&ScaleUVsByCenter}
  }
  MaterialExpressionMaterialFunctionCall_18 {
   class=MaterialExpressionMaterialFunctionCall
   inputs [
    Sine Phase {from=MaterialExpressionTime_8}
    Value 1 {from=""}
    Value 2 {from=MaterialExpressionScalarParameter_36}
   ]
   properties {materialFunction=/Engine/Functions/Engine_MaterialFunctions02/Utility/Sine_Remapped}
  }
  MaterialExpressionMaterialFunctionCall_19 {
   class=MaterialExpressionMaterialFunctionCall
   inputs [Float4 {from=MaterialExpressionMaterialFunctionCall_10, output=FX}]
   properties {materialFunction=/Game/Global/MaterialFunctions/MF_BreakComponents}
  }
  MaterialExpressionMaterialFunctionCall_2 {
   class=MaterialExpressionMaterialFunctionCall
   inputs [
    Shading_Gradient {from=MaterialExpressionMaterialFunctionCall_6, output=Shading_Gradient}
    Threshold_Gradient {from=MaterialExpressionMaterialFunctionCall_6, output=Shading_Gradient}
    Specular_Gradient {from=MaterialExpressionMaterialFunctionCall_6, output=Shading_Gradient}
    Sheen_Mask {from=MaterialExpressionMaterialFunctionCall_6, output=Shading_Gradient}
    Thickness {from=MaterialExpressionMaterialFunctionCall_6, output=Shading_Gradient}
    DFL_Mask {from=MaterialExpressionMaterialFunctionCall_6, output=Shading_Gradient}
    FX {from=MaterialExpressionMaterialFunctionCall_6, output=Shading_Gradient}
    Metallic_Gradient {from=MaterialExpressionMaterialFunctionCall_6, output=Shading_Gradient}
    RimLight_Gradient {from=MaterialExpressionMaterialFunctionCall_6, output=Shading_Gradient}
    NormalBuffer {from=MaterialExpressionMaterialFunctionCall_6, output=Shading_Gradient}
    -------- {from=""}
    CINE_LightSources {from=MaterialExpressionMaterialFunctionCall_6, output=Shading_Gradient}
    CINE_LightColor {from=MaterialExpressionMaterialFunctionCall_6, output=Shading_Gradient}
    --------- {from=""}
    Color_Lit {from=MaterialExpressionMaterialFunctionCall_11, output=Lit}
    Color_Shaded {from=MaterialExpressionMaterialFunctionCall_11, output=Lit}
    SRM {from=MaterialExpressionMaterialFunctionCall_10, output=SRM}
    Emissive {from=MaterialExpressionMaterialFunctionCall_10, output=SRM}
    DFL_Color {from=MaterialExpressionMaterialFunctionCall_10, output=SRM}
    Specular_Color {from=MaterialExpressionMaterialFunctionCall_10, output=SRM}
    RimLight_Color {from=MaterialExpressionMaterialFunctionCall_10, output=SRM}
   ]
   properties {materialFunction=/Game/Global/Materials/Character_Parent_2023/Anime_Parent_2023/MaterialFunctions/MFC_AnimeShader_2023_Post}
  }
  MaterialExpressionMaterialFunctionCall_22 {
   class=MaterialExpressionMaterialFunctionCall
   inputs [
    UVs {from=MaterialExpressionNamedRerouteUsage_46}
    CenterPosition {from=""}
    Radius {from=MaterialExpressionScalarParameter_53}
    Density {from=MaterialExpressionScalarParameter_54}
    Invert Density {from=""}
   ]
   properties {materialFunction=&RadialGradientExponential}
  }
  MaterialExpressionMaterialFunctionCall_23 {
   class=MaterialExpressionMaterialFunctionCall
   inputs [
    Vector or UVs {from=MaterialExpressionSubtract_10}
    Swizzle Coordinate Output {from=""}
   ]
   properties {materialFunction=&VectorToRadialValue}
  }
  MaterialExpressionMaterialFunctionCall_24 {
   class=MaterialExpressionMaterialFunctionCall
   inputs [
    UVs {from=MaterialExpressionFrac_0}
    Texture Scale {from=MaterialExpressionMultiply_68}
   ]
   properties {materialFunction=&ScaleUVsByCenter}
  }
  MaterialExpressionMaterialFunctionCall_25 {
   class=MaterialExpressionMaterialFunctionCall
   inputs [
    UVs {from=MaterialExpressionFrac_1}
    Texture Scale {from=MaterialExpressionMultiply_68}
   ]
   properties {materialFunction=&ScaleUVsByCenter}
  }
  MaterialExpressionMaterialFunctionCall_27 {
   class=MaterialExpressionMaterialFunctionCall
   inputs [
    UVs {from=MaterialExpressionFrac_16}
    Texture Scale {from=MaterialExpressionMultiply_68}
   ]
   properties {materialFunction=&ScaleUVsByCenter}
  }
  MaterialExpressionMaterialFunctionCall_6 {
   class=MaterialExpressionMaterialFunctionCall
   inputs [
    CustomLightVector {from=""}
    OverrideNormal_WS {from=""}
    STT {from=MaterialExpressionMaterialFunctionCall_10, output=STT}
    Normal_Map {from=MaterialExpressionMaterialFunctionCall_10, output=STT}
    DFL {from=MaterialExpressionMaterialFunctionCall_10, output=STT}
    FX {from=MaterialExpressionMaterialFunctionCall_10, output=STT}
   ]
   properties {materialFunction=/Game/Global/Materials/Character_Parent_2023/Anime_Parent_2023/MaterialFunctions/MFC_AnimeShader_2023_Calculations}
  }
  MaterialExpressionMax_0 {
   class=MaterialExpressionMax
   inputs [
    A {from=MaterialExpressionMultiply_50}
    B {from=MaterialExpressionMultiply_53}
   ]
  }
  MaterialExpressionMultiply_33 {
   class=MaterialExpressionMultiply
   inputs [
    A {from=MaterialExpressionVectorParameter_16, output=RGB}
    B {from=MaterialExpressionOneMinus_7}
   ]
  }
  MaterialExpressionMultiply_36 {
   class=MaterialExpressionMultiply
   inputs [
    A {from=MaterialExpressionMultiply_33}
    B {from=MaterialExpressionLinearInterpolate_13}
   ]
  }
  MaterialExpressionMultiply_39 {
   class=MaterialExpressionMultiply
   inputs [
    A {from=MaterialExpressionNamedRerouteUsage_42}
    B {from=MaterialExpressionComponentMask_34}
   ]
  }
  MaterialExpressionMultiply_40 {
   class=MaterialExpressionMultiply
   inputs [
    A {from=MaterialExpressionTime_9}
    B {from=MaterialExpressionComponentMask_36}
   ]
  }
  MaterialExpressionMultiply_45 {
   class=MaterialExpressionMultiply
   inputs [
    A {from=MaterialExpressionAdd_19}
    B {from=MaterialExpressionComponentMask_38}
   ]
  }
  MaterialExpressionMultiply_47 {
   class=MaterialExpressionMultiply
   inputs [
    A {from=MaterialExpressionNamedRerouteUsage_43}
    B {from=MaterialExpressionScalarParameter_43}
   ]
  }
  MaterialExpressionMultiply_48 {
   class=MaterialExpressionMultiply
   inputs [
    A {from=MaterialExpressionComponentMask_35}
    B {from=MaterialExpressionNamedRerouteUsage_35}
   ]
  }
  MaterialExpressionMultiply_49 {
   class=MaterialExpressionMultiply
   inputs [
    A {from=MaterialExpressionNamedRerouteUsage_36}
    B {from=MaterialExpressionSmoothStep_0}
   ]
  }
  MaterialExpressionMultiply_50 {
   class=MaterialExpressionMultiply
   inputs [
    A {from=MaterialExpressionComponentMask_41}
    B {from=MaterialExpressionScalarParameter_47}
   ]
  }
  MaterialExpressionMultiply_53 {
   class=MaterialExpressionMultiply
   inputs [
    A {from=MaterialExpressionMultiply_48}
    B {from=MaterialExpressionScalarParameter_48}
   ]
  }
  MaterialExpressionMultiply_54 {
   class=MaterialExpressionMultiply
   inputs [
    A {from=MaterialExpressionTextureSample_7, output=RGB}
    B {from=MaterialExpressionNamedRerouteDeclaration_25}
   ]
  }
  MaterialExpressionMultiply_55 {
   class=MaterialExpressionMultiply
   inputs [
    A {from=MaterialExpressionComponentMask_35}
    B {from=MaterialExpressionMultiply_54}
   ]
  }
  MaterialExpressionMultiply_56 {
   class=MaterialExpressionMultiply
   inputs [
    A {from=MaterialExpressionPower_6}
    B {from=MaterialExpressionScalarParameter_50}
   ]
  }
  MaterialExpressionMultiply_57 {
   class=MaterialExpressionMultiply
   inputs [
    A {from=MaterialExpressionScalarParameter_52}
    B {from=MaterialExpressionMultiply_58}
   ]
  }
  MaterialExpressionMultiply_58 {
   class=MaterialExpressionMultiply
   inputs [
    A {from=MaterialExpressionMultiply_56}
    B {from=MaterialExpressionScalarParameter_55}
   ]
  }
  MaterialExpressionMultiply_60 {
   class=MaterialExpressionMultiply
   inputs [
    A {from=MaterialExpressionVertexNormalWS_0}
    B {from=MaterialExpressionConstant_8}
   ]
   properties {constB=0.5}
  }
  MaterialExpressionMultiply_61 {
   class=MaterialExpressionMultiply
   inputs [
    A {from=MaterialExpressionTextureSample_13, output=A}
    B {from=MaterialExpressionTextureSample_13, output=A}
   ]
  }
  MaterialExpressionMultiply_62 {
   class=MaterialExpressionMultiply
   inputs [
    A {from=MaterialExpressionMultiply_61}
    B {from=MaterialExpressionMultiply_83}
   ]
  }
  MaterialExpressionMultiply_63 {
   class=MaterialExpressionMultiply
   inputs [
    A {from=MaterialExpressionTextureSample_9, output=A}
    B {from=MaterialExpressionTextureSample_9, output=A}
   ]
  }
  MaterialExpressionMultiply_64 {
   class=MaterialExpressionMultiply
   inputs [
    A {from=MaterialExpressionMultiply_65}
    B {from=MaterialExpressionMaterialFunctionCall_27, output=0-1 mask}
   ]
  }
  MaterialExpressionMultiply_65 {
   class=MaterialExpressionMultiply
   inputs [
    A {from=MaterialExpressionTextureSample_11, output=A}
    B {from=MaterialExpressionTextureSample_11, output=A}
   ]
  }
  MaterialExpressionMultiply_66 {
   class=MaterialExpressionMultiply
   inputs [
    A {from=MaterialExpressionTextureCoordinate_9}
    B {from=MaterialExpressionScalarParameter_58}
   ]
  }
  MaterialExpressionMultiply_67 {
   class=MaterialExpressionMultiply
   inputs [
    A {from=MaterialExpressionComponentMask_52}
    B {from=MaterialExpressionScalarParameter_59}
   ]
  }
  MaterialExpressionMultiply_68 {
   class=MaterialExpressionMultiply
   inputs [
    A {from=MaterialExpressionVectorParameter_23, output=RGB}
    B {from=MaterialExpressionScalarParameter_60}
   ]
  }
  MaterialExpressionMultiply_73 {
   class=MaterialExpressionMultiply
   inputs [
    A {from=MaterialExpressionComponentMask_55}
    B {from=MaterialExpressionScalarParameter_57}
   ]
   properties {constB=5, desc=Multiply the swirl amount to ensure it tiles properly with Integers}
  }
  MaterialExpressionMultiply_78 {
   class=MaterialExpressionMultiply
   inputs [
    A {from=MaterialExpressionNamedRerouteUsage_51}
    B {from=""}
   ]
   properties {constB=0.15}
  }
  MaterialExpressionMultiply_83 {
   class=MaterialExpressionMultiply
   inputs [
    A {from=MaterialExpressionMaterialFunctionCall_24, output=0-1 mask}
    B {from=MaterialExpressionStep_0}
   ]
  }
  MaterialExpressionMultiply_88 {
   class=MaterialExpressionMultiply
   inputs [
    A {from=MaterialExpressionMultiply_63}
    B {from=MaterialExpressionMaterialFunctionCall_25, output=0-1 mask}
   ]
  }
  MaterialExpressionNamedRerouteDeclaration_10 {
   class=MaterialExpressionNamedRerouteDeclaration
   inputs [None {from=MaterialExpressionLinearInterpolate_15}]
   properties {name=Swirl Portal, nodeColor=<rgba 0 1 0.2734 1>}
  }
  MaterialExpressionNamedRerouteDeclaration_13 {
   class=MaterialExpressionNamedRerouteDeclaration
   inputs [None {from=MaterialExpressionMax_0}]
   properties {name=Swirl Portal Mask, nodeColor=<rgba 1 0 0.0469 1>}
  }
  MaterialExpressionNamedRerouteDeclaration_23 {
   class=MaterialExpressionNamedRerouteDeclaration
   inputs [None {from=MaterialExpressionMaterialFunctionCall_17, output=UVs}]
   properties {name=Position UVs, nodeColor=<rgba 0 0.3672 1 1>}
  }
  MaterialExpressionNamedRerouteDeclaration_24 {
   class=MaterialExpressionNamedRerouteDeclaration
   inputs [None {from=MaterialExpressionLinearInterpolate_17}]
   properties {name=Tileable Blob, nodeColor=<rgba 0 1 0.4141 1>}
  }
  MaterialExpressionNamedRerouteDeclaration_25 {
   class=MaterialExpressionNamedRerouteDeclaration
   inputs [None {from=MaterialExpressionTextureSample_7, output=A}]
   properties {name=Tilable Blob Alpha, nodeColor=<rgba 1 0.0938 0 1>}
  }
  MaterialExpressionNamedRerouteDeclaration_26 {
   class=MaterialExpressionNamedRerouteDeclaration
   inputs [None {from=MaterialExpressionCeil_4}]
   properties {name=Light Mask, nodeColor=<rgba 0.8047 0 1 1>}
  }
  MaterialExpressionNamedRerouteDeclaration_27 {
   class=MaterialExpressionNamedRerouteDeclaration
   inputs [None {from=MaterialExpressionSaturate_11}]
   properties {name=Portal Mask, nodeColor=<rgba 1 0 0.5156 1>}
  }
  MaterialExpressionNamedRerouteDeclaration_28 {
   class=MaterialExpressionNamedRerouteDeclaration
   inputs [None {from=MaterialExpressionStaticSwitchParameter_6}]
   properties {name=Material Attributes In, nodeColor=<rgba 0 1 0.3438 1>}
  }
  MaterialExpressionNamedRerouteDeclaration_29 {
   class=MaterialExpressionNamedRerouteDeclaration
   inputs [None {from=MaterialExpressionLinearInterpolate_12}]
   properties {name=Emissive Out, nodeColor=<rgba 1 0.0234 0 1>}
  }
  MaterialExpressionNamedRerouteDeclaration_30 {
   class=MaterialExpressionNamedRerouteDeclaration
   inputs [None {from=MaterialExpressionLinearInterpolate_10}]
   properties {name=WPO, nodeColor=<rgba 0 0.2969 1 1>}
  }
  MaterialExpressionNamedRerouteDeclaration_31 {
   class=MaterialExpressionNamedRerouteDeclaration
   inputs [None {from=MaterialExpressionPanner_0}]
   properties {name=Movement UVs, nodeColor=<rgba 0 1 0.6016 1>}
  }
  MaterialExpressionNamedRerouteDeclaration_32 {
   class=MaterialExpressionNamedRerouteDeclaration
   inputs [None {from=MaterialExpressionTextureSample_6, output=R}]
   properties {name=Distortion, nodeColor=<rgba 1 0.1172 0 1>}
  }
  MaterialExpressionNamedRerouteDeclaration_33 {
   class=MaterialExpressionNamedRerouteDeclaration
   inputs [None {from=MaterialExpressionMaterialFunctionCall_19, output=A}]
   properties {name=Back Wall Mask, nodeColor=<rgba 0.3594 1 0 1>}
  }
  MaterialExpressionNamedRerouteDeclaration_34 {
   class=MaterialExpressionNamedRerouteDeclaration
   inputs [None {from=MaterialExpressionMaterialFunctionCall_22, output=RadialGradientExponential}]
   properties {name=Vignette Circle, nodeColor=<rgba 0.5938 1 0 1>}
  }
  MaterialExpressionNamedRerouteDeclaration_35 {
   class=MaterialExpressionNamedRerouteDeclaration
   inputs [None {from=MaterialExpressionLinearInterpolate_20}]
   properties {name=Ghost Phase, nodeColor=<rgba 1 0 0.5859 1>}
  }
  MaterialExpressionNamedRerouteDeclaration_36 {
   class=MaterialExpressionNamedRerouteDeclaration
   inputs [None {from=MaterialExpressionMultiply_62}]
   properties {name=Ghost 1, nodeColor=<rgba 0 0.9062 1 1>}
  }
  MaterialExpressionNamedRerouteDeclaration_37 {
   class=MaterialExpressionNamedRerouteDeclaration
   inputs [None {from=MaterialExpressionMultiply_88}]
   properties {name=Ghost 2, nodeColor=<rgba 1 0.7734 0 1>}
  }
  MaterialExpressionNamedRerouteDeclaration_38 {
   class=MaterialExpressionNamedRerouteDeclaration
   inputs [None {from=MaterialExpressionMultiply_64}]
   properties {name=Ghost 3, nodeColor=<rgba 0 1 0.7656 1>}
  }
  MaterialExpressionNamedRerouteDeclaration_39 {
   class=MaterialExpressionNamedRerouteDeclaration
   inputs [None {from=MaterialExpressionLinearInterpolate_19}]
   properties {name=Ghosts, nodeColor=<rgba 0.125 0 1 1>}
  }
  MaterialExpressionNamedRerouteDeclaration_4 {
   class=MaterialExpressionNamedRerouteDeclaration
   inputs [None {from=MaterialExpressionLinearInterpolate_0}]
   properties {name=flow map, nodeColor=<rgba 1 0 0.5156 1>}
  }
  MaterialExpressionNamedRerouteDeclaration_40 {
   class=MaterialExpressionNamedRerouteDeclaration
   inputs [None {from=MaterialExpressionDistance_0}]
   properties {name=Portal Pivot, nodeColor=<rgba 1 0.1641 0 1>}
  }
  MaterialExpressionNamedRerouteDeclaration_41 {
   class=MaterialExpressionNamedRerouteDeclaration
   inputs [None {from=MaterialExpressionLinearInterpolate_21}]
   properties {name=Ghst Step, nodeColor=<rgba 0.8047 1 0 1>}
  }
  MaterialExpressionNamedRerouteUsage_15 {class=MaterialExpressionNamedRerouteUsage}
  MaterialExpressionNamedRerouteUsage_18 {class=MaterialExpressionNamedRerouteUsage}
  MaterialExpressionNamedRerouteUsage_22 {class=MaterialExpressionNamedRerouteUsage}
  MaterialExpressionNamedRerouteUsage_23 {class=MaterialExpressionNamedRerouteUsage}
  MaterialExpressionNamedRerouteUsage_24 {class=MaterialExpressionNamedRerouteUsage}
  MaterialExpressionNamedRerouteUsage_27 {class=MaterialExpressionNamedRerouteUsage}
  MaterialExpressionNamedRerouteUsage_35 {class=MaterialExpressionNamedRerouteUsage}
  MaterialExpressionNamedRerouteUsage_36 {class=MaterialExpressionNamedRerouteUsage}
  MaterialExpressionNamedRerouteUsage_37 {class=MaterialExpressionNamedRerouteUsage}
  MaterialExpressionNamedRerouteUsage_38 {class=MaterialExpressionNamedRerouteUsage}
  MaterialExpressionNamedRerouteUsage_39 {class=MaterialExpressionNamedRerouteUsage}
  MaterialExpressionNamedRerouteUsage_40 {class=MaterialExpressionNamedRerouteUsage}
  MaterialExpressionNamedRerouteUsage_41 {class=MaterialExpressionNamedRerouteUsage}
  MaterialExpressionNamedRerouteUsage_42 {class=MaterialExpressionNamedRerouteUsage}
  MaterialExpressionNamedRerouteUsage_43 {class=MaterialExpressionNamedRerouteUsage}
  MaterialExpressionNamedRerouteUsage_44 {class=MaterialExpressionNamedRerouteUsage}
  MaterialExpressionNamedRerouteUsage_45 {class=MaterialExpressionNamedRerouteUsage}
  MaterialExpressionNamedRerouteUsage_46 {class=MaterialExpressionNamedRerouteUsage}
  MaterialExpressionNamedRerouteUsage_47 {class=MaterialExpressionNamedRerouteUsage}
  MaterialExpressionNamedRerouteUsage_48 {class=MaterialExpressionNamedRerouteUsage}
  MaterialExpressionNamedRerouteUsage_49 {class=MaterialExpressionNamedRerouteUsage}
  MaterialExpressionNamedRerouteUsage_50 {class=MaterialExpressionNamedRerouteUsage}
  MaterialExpressionNamedRerouteUsage_51 {class=MaterialExpressionNamedRerouteUsage}
  MaterialExpressionNamedRerouteUsage_52 {class=MaterialExpressionNamedRerouteUsage}
  MaterialExpressionNamedRerouteUsage_53 {class=MaterialExpressionNamedRerouteUsage}
  MaterialExpressionOneMinus_7 {class=MaterialExpressionOneMinus, inputs [None {from=MaterialExpressionFresnel_1}]}
  MaterialExpressionPanner_0 {
   class=MaterialExpressionPanner
   inputs [
    Coordinate {from=MaterialExpressionComponentMask_37}
    Time {from=""}
    Speed {from=MaterialExpressionVectorParameter_19, output=RGB}
   ]
  }
  MaterialExpressionPower_6 {
   class=MaterialExpressionPower
   inputs [
    Base {from=MaterialExpressionMultiply_55}
    Exp {from=MaterialExpressionScalarParameter_51}
   ]
  }
  MaterialExpressionPower_8 {
   class=MaterialExpressionPower
   inputs [
    Base {from=MaterialExpressionMaterialFunctionCall_23, output=Radial Coordinates}
    Exp {from=MaterialExpressionComponentMask_47}
   ]
  }
  MaterialExpressionSaturate_10 {class=MaterialExpressionSaturate, inputs [None {from=MaterialExpressionSign_1}]}
  MaterialExpressionSaturate_11 {class=MaterialExpressionSaturate, inputs [None {from=MaterialExpressionCeil_3}]}
  MaterialExpressionSaturate_12 {class=MaterialExpressionSaturate, inputs [None {from=MaterialExpressionNamedRerouteUsage_18}]}
  MaterialExpressionSaturate_13 {class=MaterialExpressionSaturate, inputs [None {from=MaterialExpressionMultiply_50}]}
  MaterialExpressionScalarParameter_34 {class=MaterialExpressionScalarParameter, properties {defaultValue=0.4292, group=KneelReedy - Backpack, parameterName=Portal - Scale}}
  MaterialExpressionScalarParameter_35 {class=MaterialExpressionScalarParameter, properties {defaultValue=0.0157, group=KneelReedy - Backpack, parameterName=Reactivity - Light Exp}}
  MaterialExpressionScalarParameter_36 {class=MaterialExpressionScalarParameter, properties {defaultValue=50, group=KneelReedy - Backpack, parameterName=Reactivity - Light}}
  MaterialExpressionScalarParameter_40 {class=MaterialExpressionScalarParameter, properties {group=KneelReedy - Backpack, parameterName=Reactivity - Door Open}}
  MaterialExpressionScalarParameter_41 {class=MaterialExpressionScalarParameter, properties {defaultValue=1, group=KneelReedy - Backpack, parameterName=Is Reactive}}
  MaterialExpressionScalarParameter_42 {
   class=MaterialExpressionScalarParameter
   properties {defaultValue=-0.05, group=KneelReedy - Backpack, parameterName=Swirl - Distortion Speed}
  }
  MaterialExpressionScalarParameter_43 {
   class=MaterialExpressionScalarParameter
   properties {defaultValue=0.03, group=KneelReedy - Backpack, parameterName=Portal - Swirl Distortion}
  }
  MaterialExpressionScalarParameter_44 {
   class=MaterialExpressionScalarParameter
   properties {defaultValue=0.0966, group=KneelReedy - Backpack, parameterName=Portal - Swirl Fade Step}
  }
  MaterialExpressionScalarParameter_45 {
   class=MaterialExpressionScalarParameter
   properties {defaultValue=0.096, group=KneelReedy - Backpack, parameterName=Portal - Swirl Fade Smoothness}
  }
  MaterialExpressionScalarParameter_46 {class=MaterialExpressionScalarParameter, properties {defaultValue=-5, group=KneelReedy - Backpack, parameterName=Portal - Mip Bias}}
  MaterialExpressionScalarParameter_47 {class=MaterialExpressionScalarParameter, properties {defaultValue=1, group=KneelReedy - Backpack, parameterName=Ghost Opacity}}
  MaterialExpressionScalarParameter_48 {class=MaterialExpressionScalarParameter, properties {group=KneelReedy - Backpack, parameterName=Portal Blob - Opacity}}
  MaterialExpressionScalarParameter_49 {class=MaterialExpressionScalarParameter, properties {group=KneelReedy - Backpack, parameterName=Is Ghost?}}
  MaterialExpressionScalarParameter_50 {class=MaterialExpressionScalarParameter, properties {defaultValue=1, group=KneelReedy - Backpack, parameterName=Swirl Color - Tint}}
  MaterialExpressionScalarParameter_51 {class=MaterialExpressionScalarParameter, properties {defaultValue=0.8, group=KneelReedy - Backpack, parameterName=Swirl Color - Contrast}}
  MaterialExpressionScalarParameter_52 {class=MaterialExpressionScalarParameter, properties {defaultValue=0.4, group=KneelReedy - Backpack, parameterName=Vignette Strength}}
  MaterialExpressionScalarParameter_53 {class=MaterialExpressionScalarParameter, properties {defaultValue=0.5, group=KneelReedy - Backpack, parameterName=Vignette Radius}}
  MaterialExpressionScalarParameter_54 {class=MaterialExpressionScalarParameter, properties {defaultValue=2.33, group=KneelReedy - Backpack, parameterName=Vignette Hardness}}
  MaterialExpressionScalarParameter_55 {class=MaterialExpressionScalarParameter, properties {defaultValue=0.9, group=KneelReedy - Backpack, parameterName=Swirl Color - Dim}}
  MaterialExpressionScalarParameter_57 {class=MaterialExpressionScalarParameter, properties {defaultValue=6, group=KneelReedy - Backpack, parameterName=Ghost - Swirl Exp}}
  MaterialExpressionScalarParameter_58 {class=MaterialExpressionScalarParameter, properties {defaultValue=3, group=KneelReedy - Backpack, parameterName=Ghost Distortion - Tiling}}
  MaterialExpressionScalarParameter_59 {class=MaterialExpressionScalarParameter, properties {defaultValue=0.02, group=KneelReedy - Backpack, parameterName=Ghost Distortion Amount}}
  MaterialExpressionScalarParameter_60 {class=MaterialExpressionScalarParameter, properties {defaultValue=1.2, group=KneelReedy - Backpack, parameterName=Ghost Overall Scale}}
  MaterialExpressionScalarParameter_61 {class=MaterialExpressionScalarParameter, properties {group=KneelReedy - Backpack, parameterName=Ghost Phase}}
  MaterialExpressionScalarParameter_63 {class=MaterialExpressionScalarParameter, properties {group=KneelReedy - Backpack, parameterName=Ghost Phase - Start Pos}}
  MaterialExpressionScalarParameter_64 {class=MaterialExpressionScalarParameter, properties {defaultValue=3, group=KneelReedy - Backpack, parameterName=Ghost Phase - Exp}}
  MaterialExpressionScalarParameter_74 {class=MaterialExpressionScalarParameter, properties {group=KneelReedy - Backpack, parameterName=Ghost Phase}}
  MaterialExpressionSetMaterialAttributes_2 {
   class=MaterialExpressionSetMaterialAttributes
   inputs [
    MaterialAttributes {from=MaterialExpressionNamedRerouteUsage_39, output=""}
    World Position Offset {from=MaterialExpressionNamedRerouteUsage_41, output=""}
    Opacity {from=MaterialExpressionSaturate_12, output=""}
    Emissive Color {from=MaterialExpressionNamedRerouteUsage_40, output=""}
    Opacity Mask {from=MaterialExpressionSaturate_12, output=""}
    Base Color {from=MaterialExpressionConstant_7, output=""}
   ]
   properties {attributeSetTypes [
     #3
     #4
     #1
     #5
     #2
    ]}
  }
  MaterialExpressionSetMaterialAttributes_3 {
   class=MaterialExpressionSetMaterialAttributes
   inputs [
    MaterialAttributes {from=MaterialExpressionGetMaterialAttributes_0, output=MaterialAttributes}
    Base Color {from=MaterialExpressionMaterialFunctionCall_19, output=RGB}
    Emissive Color {from=MaterialExpressionMaterialFunctionCall_19, output=RGB}
   ]
   properties {attributeSetTypes [#2, #1]}
  }
  MaterialExpressionSign_1 {class=MaterialExpressionSign, inputs [None {from=MaterialExpressionSubtract_3, output=""}], properties {}}
  MaterialExpressionSmoothStep_0 {
   class=MaterialExpressionSmoothStep
   inputs [
    Min {from=MaterialExpressionScalarParameter_44, output=""}
    Max {from=MaterialExpressionAdd_21, output=""}
    Value {from=MaterialExpressionComponentMask_42, output=""}
   ]
   properties {}
  }
  MaterialExpressionStaticBoolParameter_1 {
   class=MaterialExpressionStaticBoolParameter
   inputs []
   properties {defaultValue=true, group=01_EyeAnimation_Parameters, parameterName=Use UV2 for Eye Animation, sortPriority=-30}
  }
  MaterialExpressionStaticSwitchParameter_3 {
   class=MaterialExpressionStaticSwitchParameter
   inputs [
    True {from=MaterialExpressionChannelMaskParameter_1, output=R}
    False {from=MaterialExpressionSaturate_10, output=""}
   ]
   properties {group=01_EyeAnimation_Parameters, parameterName=UseEyeMask, sortPriority=-33}
  }
  MaterialExpressionStaticSwitchParameter_5 {
   class=MaterialExpressionStaticSwitchParameter
   inputs [
    True {from=MaterialExpressionAppendVector_1, output=""}
    False {from=MaterialExpressionTextureSampleParameter2D_1, output=RGBA}
   ]
   properties {group=01_EyeAnimation_Parameters, parameterName=UseVertexColorForEyeMask, sortPriority=-33}
  }
  MaterialExpressionStaticSwitchParameter_6 {
   class=MaterialExpressionStaticSwitchParameter
   inputs [
    True {from=MaterialExpressionSetMaterialAttributes_3, output=""}
    False {from=MaterialExpressionGetMaterialAttributes_0, output=MaterialAttributes}
   ]
   properties {parameterName=Debug TA Masks}
  }
  MaterialExpressionStep_0 {
   class=MaterialExpressionStep
   inputs [
    Y {from=MaterialExpressionNamedRerouteUsage_52, output=""}
    X {from=MaterialExpressionNamedRerouteUsage_53, output=""}
   ]
   properties {}
  }
  MaterialExpressionStep_1 {
   class=MaterialExpressionStep
   inputs [
    Y {from=MaterialExpressionConstant_1, output=""}
    X {from=MaterialExpressionTextureSample_0, output=R}
   ]
   properties {constX=0.2, constY=0.98}
  }
  MaterialExpressionSubtract_10 {
   class=MaterialExpressionSubtract
   inputs [
    A {from=MaterialExpressionTextureCoordinate_8}
    B {from=MaterialExpressionComponentMask_46}
   ]
   properties {constB=0.5}
  }
  MaterialExpressionSubtract_11 {
   class=MaterialExpressionSubtract
   inputs [
    A {from=MaterialExpressionAppendVector_2}
    B {from=MaterialExpressionComponentMask_53}
   ]
  }
  MaterialExpressionSubtract_12 {
   class=MaterialExpressionSubtract
   inputs [
    A {from=MaterialExpressionAdd_25}
    B {from=MaterialExpressionNamedRerouteUsage_51}
   ]
   properties {constB=0}
  }
  MaterialExpressionSubtract_13 {
   class=MaterialExpressionSubtract
   inputs [
    A {from=MaterialExpressionAdd_25}
    B {from=MaterialExpressionNamedRerouteUsage_51}
   ]
  }
  MaterialExpressionSubtract_14 {
   class=MaterialExpressionSubtract
   inputs [
    A {from=MaterialExpressionAdd_25}
    B {from=MaterialExpressionNamedRerouteUsage_51}
   ]
  }
  MaterialExpressionSubtract_3 {
   class=MaterialExpressionSubtract
   inputs [
    A {from=MaterialExpressionComponentMask_4}
    B {from=""}
   ]
   properties {constB=0.0001}
  }
  MaterialExpressionSubtract_4 {
   class=MaterialExpressionSubtract
   inputs [
    A {from=MaterialExpressionMaterialFunctionCall_19, output=G}
    B {from=MaterialExpressionMaterialFunctionCall_19, output=G}
   ]
  }
  MaterialExpressionSubtract_5 {
   class=MaterialExpressionSubtract
   inputs [
    A {from=MaterialExpressionMaterialFunctionCall_19, output=B}
    B {from=MaterialExpressionMaterialFunctionCall_19, output=B}
   ]
  }
  MaterialExpressionSubtract_8 {
   class=MaterialExpressionSubtract
   inputs [
    A {from=MaterialExpressionConvert_3}
    B {from=MaterialExpressionDivide_1}
   ]
  }
  MaterialExpressionTemporalResponsivenessOutput_0 {class=MaterialExpressionTemporalResponsivenessOutput, inputs [None {from=MaterialExpressionConstant_6, output=""}], properties {}}
  MaterialExpressionTextureCoordinate_5 {class=MaterialExpressionTextureCoordinate, properties {coordinateIndex=1}}
  MaterialExpressionTextureCoordinate_7 {class=MaterialExpressionTextureCoordinate}
  MaterialExpressionTextureCoordinate_8 {class=MaterialExpressionTextureCoordinate}
  MaterialExpressionTextureCoordinate_9 {class=MaterialExpressionTextureCoordinate}
  MaterialExpressionTextureSampleParameter2D_1 {
   class=MaterialExpressionTextureSampleParameter2D
   inputs [
    UVs {from="", output=""}
    Apply View MipBias {from="", output=""}
   ]
   properties {
    group=01_EyeAnimation_Parameters
    parameterName=EyeMask
    samplerSource=SSM_Wrap_WorldGroupSettings
    samplerType=SAMPLERTYPE_Masks
    sortPriority=-32
    texture=&T_WhiteMask
   }
  }
  MaterialExpressionTextureSample_0 {
   class=MaterialExpressionTextureSample
   inputs [
    UVs {from=MaterialExpressionNamedRerouteUsage_23}
    Tex {from=""}
    Apply View MipBias {from=""}
   ]
   properties {
    samplerSource=SSM_Clamp_WorldGroupSettings
    samplerType=SAMPLERTYPE_Grayscale
    texture=/BRCosmetics/Accessories/FORT_Backpacks/Backpack_M_MED_KneelReedy/Textures/T_TunnelUV_Mask
   }
  }
  MaterialExpressionTextureSample_11 {
   class=MaterialExpressionTextureSample
   inputs [
    UVs {from=MaterialExpressionMaterialFunctionCall_27, output=UVs}
    Tex {from=""}
    Apply View MipBias {from=""}
   ]
   properties {
    samplerSource=SSM_Wrap_WorldGroupSettings
    texture=/BRCosmetics/Accessories/FORT_Backpacks/Backpack_M_MED_KneelReedy/Textures/T_KneelSprite_01
   }
  }
  MaterialExpressionTextureSample_13 {
   class=MaterialExpressionTextureSample
   inputs [
    UVs {from=MaterialExpressionMaterialFunctionCall_24, output=UVs}
    Tex {from=""}
    Apply View MipBias {from=""}
   ]
   properties {
    samplerSource=SSM_Wrap_WorldGroupSettings
    texture=/BRCosmetics/Accessories/FORT_Backpacks/Backpack_M_MED_KneelReedy/Textures/T_KneelSprite_02
   }
  }
  MaterialExpressionTextureSample_2 {
   class=MaterialExpressionTextureSample
   inputs [
    UVs {from=MaterialExpressionNamedRerouteUsage_23}
    Tex {from=""}
    Apply View MipBias {from=""}
   ]
   properties {
    samplerSource=SSM_Clamp_WorldGroupSettings
    samplerType=SAMPLERTYPE_LinearColor
    texture=/BRCosmetics/Accessories/FORT_Backpacks/Backpack_M_MED_KneelReedy/Textures/T_TunnelUVs
   }
  }
  MaterialExpressionTextureSample_6 {
   class=MaterialExpressionTextureSample
   inputs [
    UVs {from=MaterialExpressionAdd_17}
    Tex {from=""}
    Apply View MipBias {from=""}
   ]
   properties {
    samplerSource=SSM_Wrap_WorldGroupSettings
    samplerType=SAMPLERTYPE_Grayscale
    texture=/BRCosmetics/Accessories/FORT_Backpacks/Backpack_M_MED_KneelReedy/Textures/T_PerlinNoise
   }
  }
  MaterialExpressionTextureSample_7 {
   class=MaterialExpressionTextureSample
   inputs [
    UVs {from=MaterialExpressionAdd_18}
    Tex {from=""}
    Bias {from=MaterialExpressionScalarParameter_46}
    Apply View MipBias {from=""}
   ]
   properties {mipValueMode=TMVM_MipBias, texture=/BRCosmetics/Accessories/FORT_Backpacks/Backpack_M_MED_KneelReedy/Textures/T_GraphicStrip}
  }
  MaterialExpressionTextureSample_8 {
   class=MaterialExpressionTextureSample
   inputs [
    UVs {from=MaterialExpressionMultiply_66}
    Tex {from=""}
    Apply View MipBias {from=""}
   ]
   properties {texture=/BRCosmetics/Accessories/FORT_Backpacks/Backpack_M_MED_KneelReedy/Textures/UVCurlNoise}
  }
  MaterialExpressionTextureSample_9 {
   class=MaterialExpressionTextureSample
   inputs [
    UVs {from=MaterialExpressionMaterialFunctionCall_25, output=UVs}
    Tex {from=""}
    Apply View MipBias {from=""}
   ]
   properties {
    samplerSource=SSM_Wrap_WorldGroupSettings
    texture=/BRCosmetics/Accessories/FORT_Backpacks/Backpack_M_MED_KneelReedy/Textures/T_KneelSprite_03
   }
  }
  MaterialExpressionTime_8 {class=MaterialExpressionTime}
  MaterialExpressionTime_9 {class=MaterialExpressionTime}
  MaterialExpressionVectorParameter_16 {
   class=MaterialExpressionVectorParameter
   properties {group=KneelReedy - Backpack, parameterName=Reactivity - Light Color, defaultValue=<rgba 1 0.2651 0.0938 1>}
  }
  MaterialExpressionVectorParameter_17 {
   class=MaterialExpressionVectorParameter
   properties {group=KneelReedy - Backpack, parameterName=Swirl - Distortion Tiling, defaultValue=<rgba 5.9812 5.9812 0 0>}
  }
  MaterialExpressionVectorParameter_18 {
   class=MaterialExpressionVectorParameter
   properties {group=KneelReedy - Backpack, parameterName=Portal - Swirl Tiling, defaultValue=<rgba 2 1 0 0>}
  }
  MaterialExpressionVectorParameter_19 {
   class=MaterialExpressionVectorParameter
   properties {group=KneelReedy - Backpack, parameterName=Portal - Swirl Speed, defaultValue=<rgba -0.03 0 0 0>}
  }
  MaterialExpressionVectorParameter_21 {
   class=MaterialExpressionVectorParameter
   properties {group=KneelReedy - Backpack, parameterName=Portal - Swirl Offset, defaultValue=<rgba 0 -0.1052 0 0>}
  }
  MaterialExpressionVectorParameter_22 {
   class=MaterialExpressionVectorParameter
   properties {group=KneelReedy - Backpack, parameterName=Portal - Position, defaultValue=<rgba -0.2912 0.2672 0 0>}
  }
  MaterialExpressionVectorParameter_23 {
   class=MaterialExpressionVectorParameter
   properties {group=KneelReedy - Backpack, parameterName=Ghosts - X/Y Scale, defaultValue=<rgba 0.3 0.5 0 0>}
  }
  MaterialExpressionVectorParameter_25 {
   class=MaterialExpressionVectorParameter
   properties {group=KneelReedy - Backpack, parameterName=Ghost Pivot, defaultValue=<rgba 0.788 0.2216 0 0>}
  }
  MaterialExpressionVectorParameter_28 {
   class=MaterialExpressionVectorParameter
   properties {group=KneelReedy - Backpack, parameterName=Ghosts Length Exp, defaultValue=<rgba 1 0.4 0 1>}
  }
  MaterialExpressionVectorParameter_29 {
   class=MaterialExpressionVectorParameter
   properties {group=KneelReedy - Backpack, parameterName=Ghost 2 - Offset, defaultValue=<rgba 0.392 0.3904 0 0>}
  }
  MaterialExpressionVectorParameter_31 {
   class=MaterialExpressionVectorParameter
   properties {group=KneelReedy - Backpack, parameterName=Ghost 3 - Offset, defaultValue=<rgba 0.0576 -0.8628 0 0>}
  }
  MaterialExpressionVectorParameter_35 {
   class=MaterialExpressionVectorParameter
   properties {group=KneelReedy - Backpack, parameterName=Ghost Swirl - Global Offset, defaultValue=<rgba 0 0.1 0 0>}
  }
  MaterialExpressionVectorParameter_36 {
   class=MaterialExpressionVectorParameter
   properties {group=KneelReedy - Backpack, parameterName=Ghost 1 - Offset, defaultValue=<rgba -0.4456 0.6115 0 0>}
  }
  MaterialExpressionVertexColor_1 {class=MaterialExpressionVertexColor, inputs [], properties {}}
  MaterialExpressionVertexNormalWS_0 {class=MaterialExpressionVertexNormalWS, inputs [], properties {}}
 }
 properties {
  bAutomaticallySetUsageInEditor=false
  bHasPixelAnimation=true
  bOutputTranslucentVelocity=true
  bUseMaterialAttributes=true
  bUsedWithMeshDeformer=true
  bUsedWithMorphTargets=true
  bUsedWithSkeletalMesh=true
  bUsedWithStaticLighting=true
  pixelDepthOffsetMode=PDOM_Legacy
  shadingModel=MSM_Cloth
  usedShadingModels=Cloth
 }
}
